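# Physics asset for the MicroLeaf backpack cosmetic mesh, giving its dangling body, head and tail parts simulated motion.
# Capsule bodies cover the pack, body, head and three tail bones, chained from pack to body, with head and tail branching off the body.
# Body and head joints lock twist and allow only minimal swing, while the tail segments get progressively looser swing limits with slight twist for subtle wag.
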PhysicsAsset /BRCosmetics/Accessories/FORT_Backpacks/Backpack_M_MED_MicroLeaf/Meshes/M_MED_MicroLeaf_Pack_Physics
bodies {
 dyn_body [{shapeType=Capsule, length=5, radius=5, shapeName=dyn_body_capsule, center=<1.2768 0 -4.342>, rotation=<rot 90 0 0>}]
 dyn_head [{shapeType=Capsule, length=4, radius=4, shapeName=dyn_body_capsule, center=<2.8824 0 2.6631>, rotation=<rot 84.2661 0 0>}]
 dyn_tail_01 [{shapeType=Capsule, length=1, radius=1.5, shapeName=dyn_body_capsule, center=<1.3284 0 0>, rotation=<rot 90 0 0>}]
 dyn_tail_02 [{shapeType=Capsule, length=1, radius=1.5, shapeName=dyn_body_capsule, center=<1.3284 0 0>, rotation=<rot 90 0 0>}]
 dyn_tail_03 [{shapeType=Capsule, length=1, radius=1.5, shapeName=dyn_body_capsule, center=<1.3284 0 0>, rotation=<rot 90 0 0>}]
 pack [{shapeType=Capsule, length=5, radius=3, shapeName=pack_capsule, rotation=<rot 0 -90 0>}]
}
constraints [
 {bone1Name=dyn_body, bone2Name=pack, swing1LimitDegrees=1, swing2LimitDegrees=1, twistLimitDegrees=1, twistMotion=Locked}
 {bone1Name=dyn_head, bone2Name=dyn_body, swing1LimitDegrees=1, swing2LimitDegrees=1, twistLimitDegrees=1, twistMotion=Locked}
 {bone1Name=dyn_tail_01, bone2Name=dyn_body, swing1LimitDegrees=5, swing2LimitDegrees=10, twistLimitDegrees=1}
 {bone1Name=dyn_tail_02, bone2Name=dyn_tail_01, swing1LimitDegrees=10, swing2LimitDegrees=15, twistLimitDegrees=1}
 {bone1Name=dyn_tail_03, bone2Name=dyn_tail_02, swing1LimitDegrees=10, swing2LimitDegrees=15, twistLimitDegrees=1}
]
