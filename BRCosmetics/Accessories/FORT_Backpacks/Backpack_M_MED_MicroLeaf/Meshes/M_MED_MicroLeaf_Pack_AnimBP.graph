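# Animation Blueprint for the MicroLeaf backpack cosmetic, adding secondary motion to its pack and tail bones.
# The AnimGraph blends two branches from a cached, nudged input pose: root-space RigidBody with AnimDynamics, and pack-space zero-gravity RigidBody with additive bone offsets.
# The EventGraph maps owner spine bend into pack rotation and translation for the second branch.
# It also sets ground and gravity variables from fall, skydive and parachute states.
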AnimBlueprint /BRCosmetics/Accessories/FORT_Backpacks/Backpack_M_MED_MicroLeaf/Meshes/M_MED_MicroLeaf_Pack_AnimBP
target_skeleton=&/M_MED_MicroLeaf_Pack_Skeleton
graphs [
 AnimGraph {nodes [
   n1 = "Transform (Modify) Bone - Bone: dyn_tail_02"!(ComponentPose, bAlphaBoolEnabled:bool=True, Translation="0,0,0", Rotation="-10,0,0", Scale="1,1,1") -> (Pose) {
    alphaInputType=Bool
    rotationMode=BMM_Additive
    rotationSpace=BCS_BoneSpace
    alphaBoolBlend {blendInTime=0.2, blendOutTime=0.2}
    boneToModify {boneName=dyn_tail_02}
    rotation {pitch=-10}
   }
   "Transform (Modify) Bone - Bone: dyn_tail_03"!(ComponentPose=@n1, bAlphaBoolEnabled:bool=True, Translation="0,0,0", Rotation="-5,0,0", Scale="1,1,1") -> (Pose) {
    alphaInputType=Bool
    rotationMode=BMM_Additive
    rotationSpace=BCS_BoneSpace
    alphaBoolBlend {blendInTime=0.2, blendOutTime=0.2}
    boneToModify {boneName=dyn_tail_03}
    rotation {pitch=-5}
   }
   n2 = RigidBody!(ComponentPose=Local To Component(LocalPose=Use cached pose 'Input_pose'()), Alpha, OverrideWorldGravity:Vector="0,0,-980", ExternalForce) -> (Pose) {
    bClampLinearTranslationLimitToRefPose=true
    bOverrideWorldGravity=true
    overridePhysicsAsset=&/M_MED_MicroLeaf_Pack_Physics
    simulationSpace=BaseBoneSpace
    baseBoneRef {boneName=root}
    componentAppliedLinearAccClamp=<0 0 0>
    simSpaceSettings {
     dampingAlpha=0
     maxAngularAcceleration=1000
     maxAngularVelocity=3
     maxLinearAcceleration=700
     maxLinearVelocity=500
     velocityScaleZ=0
     worldAlpha=0.5
     externalLinearDragV=<0.2 0.2 0.2>
    }
   }
   n3 = Anim Dynamics - Bone: dyn_tail_01!(ComponentPose=@n2, Alpha=0.25, GravityOverride:Vector="0,0,-980", ExternalForce="0,0,0") -> (Pose) {
    angularDampingOverride=10
    angularSpringConstant=1000
    bUseGravityOverride=true
    bUsePlanarLimit=false
    simulationSpace=RootRelative
    worldSpaceMinimumScale=0
    boundBone {boneName=dyn_tail_01}
    componentAppliedLinearAccClamp=<0 0 0>
    physicsBodyDefinitions [{
      boundBone {boneName=dyn_tail_01}
      boxExtents=<10 10 10>
      constraintSetup {angularLimitsMax=<5 0 15>, angularLimitsMin=<-5 0 -15>, angularTarget=<0 0 0>}
      localJointOffset=<2 0 0>
     }]
    relativeSpaceBone {boneName=root}
    simSpaceSettings {simSpaceAngularAlpha=0.3}
   }
   n4 = RigidBody(ComponentPose=Local To Component(LocalPose=Use cached pose 'Input_pose'()), OverrideWorldGravity="0,0,0") {
    bClampLinearTranslationLimitToRefPose=true
    bOverrideWorldGravity=true
    overridePhysicsAsset=&M_MED_MicroLeaf_Pack_Physics
    simulationSpace=BaseBoneSpace
    baseBoneRef {boneName=pack}
    componentAppliedLinearAccClamp=<0 0 0>
    simSpaceSettings {
     maxAngularAcceleration=1
     maxAngularVelocity=1
     maxLinearAcceleration=700
     maxLinearVelocity=500
     worldAlpha=1
     externalLinearDragV=<0.2 0.2 0.2>
    }
   }
   n5 = vector / vector(A:Rotator=Get backpack_rotate(), B:Vector="0.9,1,1") -> (ReturnValue:Vector)
   n6 = Get backpack_translation()
   n7 = "Transform (Modify) Bone - Bone: dyn_tail_02"(
    ComponentPose="Transform (Modify) Bone - Bone: pack"(
     ComponentPose=@n4
     Translation=vector / vector(A:Vector=@n6, B="1,4,2")
     Rotation=Rotation From X Vector!(InVec:Vector=@n5) -> (ReturnValue:Rotator)
    ) {rotationMode=BMM_Additive, translationMode=BMM_Additive, boneToModify {boneName=pack}}
    bAlphaBoolEnabled=True
    Translation="0,0,0"
    Rotation="-10,0,0"
    Scale="1,1,1"
   ) {
    alphaInputType=Bool
    rotationMode=BMM_Additive
    rotationSpace=BCS_BoneSpace
    alphaBoolBlend {blendInTime=0.2, blendOutTime=0.2}
    boneToModify {boneName=dyn_tail_02}
    rotation {pitch=-10}
   }
   n8 = "Transform (Modify) Bone - Bone: dyn_tail_03"(ComponentPose=@n7, bAlphaBoolEnabled=True, Translation="0,0,0", Rotation="-15,0,0", Scale="1,1,1") {
    alphaInputType=Bool
    rotationMode=BMM_Additive
    rotationSpace=BCS_BoneSpace
    alphaBoolBlend {blendInTime=0.2, blendOutTime=0.2}
    boneToModify {boneName=dyn_tail_03}
    rotation {pitch=-15}
   }
   n9 = "Transform (Modify) Bone - Bone: dyn_tail_02"(ComponentPose=@n8, bAlphaBoolEnabled=True, Translation="0,0,0", Rotation="-10,0,0", Scale="1,1,1") {
    alphaInputType=Bool
    rotationMode=BMM_Additive
    rotationSpace=BCS_BoneSpace
    alphaBoolBlend {blendInTime=0.2, blendOutTime=0.2}
    boneToModify {boneName=dyn_tail_02}
    rotation {pitch=-10}
   }
   n10 = "Transform (Modify) Bone - Bone: dyn_tail_03"(ComponentPose=@n9, bAlphaBoolEnabled=True, Translation="0,0,0", Rotation="-15,0,0", Scale="1,1,1") {
    alphaInputType=Bool
    rotationMode=BMM_Additive
    rotationSpace=BCS_BoneSpace
    alphaBoolBlend {blendInTime=0.2, blendOutTime=0.2}
    boneToModify {boneName=dyn_tail_03}
    rotation {pitch=-15}
   }
   n11 = "Transform (Modify) Bone - Bone: pack"!(ComponentPose=@n10, bAlphaBoolEnabled:bool, Translation="0,0,0", Rotation="0,0,15", Scale) -> (Pose) {
    alphaInputType=Bool
    rotationMode=BMM_Additive
    rotationSpace=BCS_BoneSpace
    alphaBoolBlend {blendInTime=0.2, blendOutTime=0.2}
    boneToModify {boneName=pack}
    rotation {roll=15}
   }
   n12 = "Transform (Modify) Bone - Bone: pack"(ComponentPose=@n11, Translation="0,0,0", Rotation="0,0,-10") {
    alphaInputType=Bool
    rotationMode=BMM_Additive
    rotationSpace=BCS_BoneSpace
    alphaBoolBlend {blendInTime=0.2, blendOutTime=0.2}
    boneToModify {boneName=pack}
    rotation {roll=-10}
   }
   n13 = "Transform (Modify) Bone - Bone: base_tail"(ComponentPose=@n12, Alpha=1, Translation="0,0,0", Rotation="-20,-20,10", Scale="1,1,1") {rotationMode=BMM_Additive, rotationSpace=BCS_BoneSpace, boneToModify {boneName=base_tail}, rotation=<rot -20 -20 10>}
   n14 = "Transform (Modify) Bone - Bone: base_tail"(ComponentPose=@n13, Alpha=1, Translation="0,0,0", Rotation="-10,-15,0", Scale="1,1,1") {rotationMode=BMM_Additive, rotationSpace=BCS_BoneSpace, boneToModify {boneName=base_tail}, rotation {pitch=-10, yaw=-15}}
   Output Pose(Result=Blend Poses by bool(BlendPose_0=Component To Local(ComponentPose=@n3), BlendPose_1=Component To Local(ComponentPose=@n14))) {layerGroup=None}
   n15 = Local To Component(LocalPose=Input Pose())
   Save cached pose 'Input_pose'(Pose=Component To Local(ComponentPose="Transform (Modify) Bone - Bone: pack"!(ComponentPose=@n15, Alpha:float, Translation="0,1,0", Rotation="0,0,0", Scale) -> (Pose) {translationMode=BMM_Additive, translationSpace=BCS_BoneSpace, boneToModify {boneName=pack}, translation {y=1}})) {cacheName=Input_pose}
   Use cached pose 'Input_pose'()
  ]}
 EventGraph {nodes [
   Event BlueprintUpdateAnimation()
   n1 = GetOwnerSkeletalMeshComponent()
   n2 = BreakRotator(InRot=BreakTransform(InTransform=GetDeltaTransformFromRefPose(self=@n1, BoneName=spine_05, BaseName=spine_01)).Rotation)
   > Set backpack_rotate(backpack_rotate=MakeRotator(Roll=MapRangeClamped(Value=@n2.Yaw, InRangeA=0, InRangeB=-30, OutRangeB=35)))
   n3 = MapRangeClamped(Value=@n2.Yaw, InRangeA=0, InRangeB=-30, OutRangeB=-10)
   > Set backpack_translation(backpack_translation=MakeVector(Y=@n3, Z=float * float(A=@n3, B=-1) {is=title K2Node_PromotableOperator}))
   > n4 = Branch(Condition=Get bIsSkydiveFloating())
   > Set gravity_override(gravity_override="0,0,980")
   n5 = Get bIsFalling()
   n6 = Get bIsInAir() -> (bIsInAir:bool)
   > Set on_the_ground(on_the_ground:bool=OR Boolean(A=NOT Boolean(A=@n5), B=NOT Boolean(A=@n6))) -> (Output_Get:bool)
   Set gravity_override(execute=@n4.else, gravity_override="0,0,-980")
   > n7 = Branch(Condition=OR Boolean(A=Get bIsParachuteOpened(), B=Get bIsSkydiveDiveMode()))
   > Set on_the_ground(on_the_ground=false)
   Set on_the_ground(execute=@n7.else, on_the_ground=Get bIsInAir())
   TryGetPawnOwner()
  ]}
]
variables [
 backpack_rotate {type=Rotator}
 backpack_translation {type=Vector}
 gravity_override {type=Vector}
 on_the_ground {type=boolean}
]
